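# Pinball vehicle wheel material that adds animated emissive lights to the shared MF_WheelParent_Simple base.
# Time-driven sine pulses, phase-offset by the AniMask green channel, light animated dots, while a second pulse lights the rim mask's green area, both tinted by base color.
# Vertex red blends from the animated result back to the unanimated base.
# The blend then passes through the shared vehicle damage and de-rez transition functions.
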Material /VehicleCosmetics/Wheels/PinBall/Materials/M_Wheel_PinBall
graph {
 properties {
  bAutomaticallySetUsageInEditor=false
  bUseMaterialAttributes=true
  bUsedWithMeshDeformer=true
  bUsedWithSkeletalMesh=true
  pixelDepthOffsetMode=PDOM_Legacy
  usedShadingModels=Default Lit
 }
 nodes [
  n1 = MaterialFunctionCall() {materialFunction=&MF_WheelParent_Simple}
  n2 = GetMaterialAttributes(None=@n1.MaterialAttributes) {attributeGetTypes [#1, #2]}
  n3 = Time()
  n4 = ScalarParameter() {defaultValue=0.5, group=0_AnimatedProperties, parameterName=SpeedDots}
  n5 = TextureSampleParameter2D() {parameterName=AniMask, samplerType=SAMPLERTYPE_LinearColor, texture=/VehicleCosmetics/Wheels/PinBall/Textures/T_Rim_Pinball_Mask}
  n6 = StaticSwitchParameter(True=@n5.G, False=@n5.G) {group=0_AnimatedProperties, parameterName=isLOD}
  n7 = Sine(None=Add(A=Multiply(A=@n3, B=@n4) {constB=0.5}, B=@n6))
  n8 = Clamp(None=OneMinus(None=ConstantBiasScale(None=@n7) {bias=1.5, scale=1}))
  n9 = MaterialFunctionCall(UV Channel=@n6) {materialFunction=&LinearGradient}
  n10 = Multiply(
   A=@n2.Base Color
   B=Multiply(
    A=Multiply(A=@n8, B=@n9.UGradient)
    B=ScalarParameter() {defaultValue=10, group=0_AnimatedProperties, parameterName=AniDotsEmissiveBrightness}
   )
  )
  n11 = Multiply(A=Time()) {constB=0.5}
  n12 = Multiply(
   A=Multiply(A=@n2.Base Color, B=ScalarParameter() {defaultValue=5, group=0_AnimatedProperties, parameterName=AniEmissiveBrightness})
   B=Clamp(None=Sine(None=@n11)) {maxDefault=2}
  )
  n13 = ComponentMask(None=@n1.RimRGBAMask) {g=true}
  n14 = SetMaterialAttributes(MaterialAttributes=@n1.MaterialAttributes, Emissive Color=Add(A=Add(A=@n2.Emissive Color, B=@n10), B=Multiply(A=@n12, B=@n13) {constB=0.5})) {attributeSetTypes [#1]}
  n15 = VertexColor()
  MaterialFunctionCall(In=MaterialFunctionCall(Material=BlendMaterialAttributes(A=@n14, B=@n1.MaterialAttributes, Alpha=@n15.R)) {materialFunction=&MF_BaseVehicle_Damage}.Result) {materialFunction=&MF_VehicleDeRez_Transition}
 ]
}
